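# Material function that produces an animated sine-wave displacement vector for the DADBRO storm skirt effect.
# A sine of V plus scrolling time is bias-scaled, then multiplied by a slower time sine that periodically reverses its direction, and scaled by DisplacementStrength.
# A clamped V gradient, built from top offset, falloff and inverted-V bottom adjustment, masks the output at both ends.
# The SineMaskFrequency parameter is unconnected.
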MaterialFunction /Game/Athena/DADBRO/FX/StormSkirt/MF_SkirtSineWave
graph {nodes [
  n1 = ComponentMask(None=TextureCoordinate()) {g=true}
  n2 = Time()
  n3 = ScalarParameter() {defaultValue=-0.25, parameterName=SineWaveSpeed}
  n4 = Multiply(A=Add(A=@n1, B=Multiply(A=@n2, B=@n3)), B=ScalarParameter() {defaultValue=2, parameterName=SineWaveFreqwuency})
  n5 = Time()
  n6 = ScalarParameter() {defaultValue=0.35, parameterName=SineDirectionChangeSpeed}
  n7 = Multiply(A=ConstantBiasScale(None=Sine(None=@n4)), B=Sine(None=Multiply(A=@n5, B=@n6)))
  n8 = VectorParameter() {parameterName=DisplacementStrength, defaultValue=<rgba 512 512 0 1>}
  n9 = Clamp(None=Subtract(A=@n1, B=ScalarParameter() {defaultValue=0.3, parameterName=SkirtTopAdjustment}))
  n10 = ScalarParameter() {defaultValue=2, parameterName=SkirtGradienFalloff}
  n11 = Multiply(
   A=Power(Base=Multiply(A=@n9, B=@n10) {constB=5})
   B=Multiply(A=OneMinus(None=@n1), B=ScalarParameter() {defaultValue=3, parameterName=SkirtBottomGradientAdjustment})
  )
  FunctionOutput(None=Multiply(A=Multiply(A=@n7, B=@n8.RGB), B=Clamp(None=@n11)))
  ScalarParameter() {parameterName=SineMaskFrequency}
 ]}
