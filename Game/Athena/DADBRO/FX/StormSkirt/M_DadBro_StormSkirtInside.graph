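# Translucent lit material for the inside of the DadBro storm skirt on skeletal meshes, drawing animated, smoky storm cloud.
# Panned Perlin noise, a smoke tile and a storm-edge mask combine with axis-aligned Fresnel and vertex red into dithered opacity.
# A curve-atlas color driven by that mask is shaded through a panning explosion normal for view-dependent darkening and emissive.
# Noise plus a skirt sine-wave function builds vertex displacement, alongside an inline gradient-masked sine sway.
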Material /Game/Athena/DADBRO/FX/StormSkirt/M_DadBro_StormSkirtInside
graph {
 properties {
  bUsedWithMeshDeformer=true
  bUsedWithSkeletalMesh=true
  blendMode=BLEND_Translucent
  pixelDepthOffsetMode=PDOM_Legacy
  translucencyLightingMode=TLM_VolumetricPerVertexDirectional
  usedShadingModels=Default Lit
 }
 nodes [
  n1 = TextureCoordinate()
  n2 = ComponentMask(None=VectorParameter() {parameterName=Noise_01Tiling, defaultValue=<rgba 1 3.5 0 1>}.RGB) {g=true, r=true}
  n3 = VectorParameter() {parameterName=Noise_01_PanSpeedAndDirection, defaultValue=<rgba 0.5 0.5 0 1>}
  n4 = TextureSample(UVs=Panner(Coordinate=Multiply(A=@n1, B=@n2), Speed=ComponentMask(None=@n3.RGB) {g=true, r=true}) {bFractionalPart=true, speedX=0.5, speedY=0.5}) {texture=/Game/Global/Textures/Default/Noise/3xPerlin_M}
  n5 = TextureCoordinate()
  n6 = ComponentMask(None=VectorParameter() {parameterName=Noise_02Tiling, defaultValue=<rgba 1 3 0 1>}.RGB) {g=true, r=true}
  n7 = VectorParameter() {parameterName=Noise_02_PanSpeedAndDirection, defaultValue=<rgba 1 0.8 0 1>}
  n8 = TextureSample(UVs=Panner(Coordinate=Multiply(A=@n5, B=@n6), Speed=ComponentMask(None=@n7.RGB) {g=true, r=true}) {bFractionalPart=true, speedX=1, speedY=0.8}) {texture=&3xPerlin_M}
  Add(
   A=Multiply(A=Multiply(A=@n4.RGB, B=@n8.RGB), B=ScalarParameter() {defaultValue=40, parameterName=WPO_Intensity}) {constB=40}
   B=MaterialFunctionCall() {materialFunction=&/MF_SkirtSineWave}.Result
  )
  n9 = MaterialFunctionCall(Float2=Saturate(None=TextureCoordinate() {vTiling=6.5})) {materialFunction=&BreakOutFloat2Components}
  n10 = OneMinus(None=MaterialFunctionCall(Float2=TextureCoordinate()) {materialFunction=&BreakOutFloat2Components}.G)
  n11 = TextureCoordinate()
  n12 = Subtract(
   A=Multiply(A=Multiply(A=@n9.G), B=Multiply(A=@n10, B=@n10))
   B=TextureSample(UVs=Panner(Coordinate=@n11) {speedX=2, speedY=0.3}) {samplerType=SAMPLERTYPE_LinearColor, texture=/Game/Global/Textures/Default/Tileable_Patterns/T_Tile_Smoke_Small}.R
  )
  n13 = Panner(Coordinate=TextureCoordinate()) {speedX=1}
  n14 = Multiply(
   A=Saturate(None=Add(A=@n12) {constB=0.8})
   B=TextureSampleParameter2D(UVs=MaterialFunctionCall(XY=@n13) {materialFunction=&Swizzle}.YX) {parameterName=Mask, texture=/Game/Effects/Fort_Effects/Textures/Mask/T_StormEdgeMask}.R
  )
  Add(A=@n14, B=@n4.R)
  Clamp(None=DotProduct(A=CameraVectorWS(), B=VertexNormalWS()))
  LinearInterpolate(
   A=VectorParameter() {parameterName=Color1, defaultValue=<rgba 0.0069 0.0093 0.0156 1>}.RGB
   B=VectorParameter() {parameterName=Color2, defaultValue=<rgba 0.0652 0.1041 0.1406 1>}.RGB
   Alpha=@n4.R
  )
  n15 = ScalarParameter() {defaultValue=1, parameterName=EdgeFadeExp}
  n16 = Multiply(
   A=Multiply(
    A=MaterialFunctionCall!(Axis, Exponent=@n15, Normal) {materialFunction=/Game/Effects/Fort_Effects/Materials/_Functions/MF_AxisAlignedFrensnel}.Fresnel
    B=VertexColor().R
   )
   B=@n14
  )
  n17 = Clamp(None=Add(A=@n16, B=Add(A=@n4.G, B=@n8.R)))
  MaterialFunctionCall(Alpha Threshold=Multiply(A=Multiply(A=@n16, B=@n17), B=ScalarParameter() {defaultValue=1, parameterName=OpacityOverall})) {materialFunction=&DitherTemporalAA}
  n18 = CameraVectorWS()
  n19 = TextureCoordinate()
  n20 = ComponentMask(None=VectorParameter() {parameterName=NormalTiling, defaultValue=<rgba 1 3 0 1>}.RGB) {g=true, r=true}
  n21 = VectorParameter() {parameterName=NormalPanSpeedAndDirection, defaultValue=<rgba 1 3 0 1>}
  n22 = TextureSample(UVs=Panner(Coordinate=Multiply(A=@n19, B=@n20), Speed=ComponentMask(None=@n21.RGB) {g=true, r=true}) {speedX=1, speedY=0.3}) {samplerType=SAMPLERTYPE_Normal, texture=/Game/Effects/Fort_Effects/Gameplay/SmokeDustExplosions/Textures/T_Explosion_NM_01}
  n23 = Transform(None=@n22.RGB)
  n24 = CurveAtlasRowParameter(CurveTime=Clamp(None=@n14)) {
   atlas=&AthenaCurveAtlas01
   curve=/Game/Athena/LinearColorCurves/Curve_StormSkirt
   defaultValue=149
   parameterName=ColorCUrve
  }
  Multiply(A=Clamp(None=DotProduct(A=@n18, B=@n23), Min=ScalarParameter() {defaultValue=0.5, parameterName=MinDarkness}), B=@n24)
  n25 = ComponentMask(None=Transform(None=ReflectionVectorWS(CustomWorldNormal=@n22.RGB)) {transformType=TRANSFORM_View}) {g=true}
  Multiply(A=Multiply(A=LinearInterpolate(Alpha=@n25) {constA=0.7}, B=@n24), B=ScalarParameter() {defaultValue=0.6, parameterName=EmissiveIntensity})
  n26 = ComponentMask(None=TextureCoordinate()) {g=true}
  n27 = Add(A=@n26, B=Multiply(A=Time(), B=ScalarParameter() {defaultValue=-0.25, parameterName=SineWaveSpeed}))
  n28 = ScalarParameter() {defaultValue=2, parameterName=SineWaveFreqwuency}
  n29 = ConstantBiasScale(None=Sine(None=Multiply(A=@n27, B=@n28)))
  n30 = Sine(None=Multiply(A=Time(), B=ScalarParameter() {defaultValue=0.35, parameterName=SineDirectionChangeSpeed}))
  n31 = Subtract(A=@n26, B=ScalarParameter() {defaultValue=0.3, parameterName=SkirtTopAdjustment})
  n32 = Power(Base=Multiply(A=Clamp(None=@n31), B=ScalarParameter() {defaultValue=2, parameterName=SkirtGradienFalloff}) {constB=5})
  n33 = Multiply(A=OneMinus(None=@n26), B=ScalarParameter() {defaultValue=3, parameterName=SkirtBottomGradientAdjustment})
  Multiply(
   A=Multiply(A=Multiply(A=@n29, B=@n30), B=VectorParameter() {parameterName=DisplacementStrength, defaultValue=<rgba 512 512 0 1>}.RGB)
   B=Clamp(None=Multiply(A=@n32, B=@n33))
  )
  Reroute(None=@n22.RGB)
  ScalarParameter() {parameterName=SineMaskFrequency}
 ]
}
